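; War Effort funding station prop, a base Blueprint where players donate toward a goal or an A/B vote, with child classes overriding state and vote events.
; After match ready, calendar events and hotfix floats drive setup; replicated Stage selects nonexistent, fundraising or funded.
; Donations are scaled and dispatched; returned totals trigger votes or goal checks, vending reward loot below goal.
; Replicated counters fire local GameplayCues for donation and goal feedback.
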(event Custom|CheckHotfixFunded
  ;[1 1 1] Check Hotfix to autofund
  (if (> (ScalableFloat|GetValueatLevel (Variables|Default|GetFundedHotfix)) 0.0)
    (CallFunction|FundedState)
    (elif (> (ScalableFloat|GetValueatLevel (Variables|Default|GetFundedAHotfix)) 0.0)
      (CallFunction|FundedStateA)
      (elif (> (ScalableFloat|GetValueatLevel (Variables|Default|GetFundedBHotfix)) 0.0)
        (CallFunction|FundedStateB)))))

(event Custom|BindCalendarEvent
  ;[1 1 1] Comment
  (if (Variables|Default|GetShouldSkipCalendarEvent)
    (Variables|Default|SetStage 1))
  (bind _array_element (Utilities|Array|ForEachLoopwithBreak (Variables|Default|GetSetupEvents))
    (:LoopBody
      (bind _eventname (MCPContext|WatchCalendarEvent _array_element)
        (:OnEventActive
          (Variables|Default|SetCurrentActiveEvent _eventname)
          (CallFunction|SetupUnvaultItems))))
    (:Completed
      _))
  (if (Utilities|Array|ContainsItem (Variables|Default|GetSetupEvents) (Variables|Default|GetCurrentActiveEvent))
    (CallFunction|CheckHotfixFunded)
    (else
      (Variables|Default|SetStage 0))))

(event Custom|FundraisingState)

(event Custom|SetupUnvaultItems)

(event Custom|FundedStateA)

(event Custom|FundedState)

(event Custom|FundedStateB)

(event Custom|NonexistentState
  ;[1 1 1] Sates to override in childs
  (Actor|DestroyActor))

(event Custom|VoteA
  ;[1 1 1] Override based on choice
  (Development|PrintString "I VOTED A"))

(event Custom|VoteB
  ;[1 1 1] Override based on choice
  (Development|PrintString "I VOTED B"))

(event Custom|CheckWeaponsFunded)

(event EventBeginPlay
  ;[1 1 1] On Match ready Initalize calendar events
  (|CallParentFunction)
  (if (> (ScalableFloat|GetValueatLevel (Variables|Default|GetHotFixOff)) 0.0)
    (Variables|Default|SetStage 0)
    (else
      (Networking|SwitchHasAuthority
        (:Authority
          (Class|FortAsyncActionAuthorityMatchReady|MatchReadyAsync
            (:OnReady
              (CallFunction|BindCalendarEvent))))
        (:Remote
          (Variables|Default|SetInitial%20Relative%20Rotation (Variables|/Script/Engine.SceneComponent|GetRelativeRotation (Variables|Debug|GetStaticMeshComponent))))))))

(event Custom|CheckFundingAmount
  ;[1 1 1] Check Funding amount and change states if funding is achieved
  (if (>= (Variables|Funding|GetTotalFunds) (Math|Float|TruncatetoInteger64 (ScalableFloat|GetValueatLevel (Variables|Funding|GetMaxFundsRequired))))
    (if (not (Variables|Default|GetIsFunded))
      (Variables|Default|SetIsFunded true)
      (CallFunction|FundingAchieved))
    (else
      (CallFunction|FundsAdded))))

(event Custom|EdgeServerShowFunding
  (params (TotalFunds :type Integer64) (IsVoteA :type Boolean) (IsOneVote :type Boolean))
  ;[1 1 1] On Edge Servers check funding amount for single and a or b choices
  (Variables|Funding|SetTotalFunds TotalFunds)
  (if IsOneVote
    (CallFunction|CheckFundingAmount)
    (elif IsVoteA
      (CallFunction|VoteA)
      (else
        (CallFunction|VoteB)))))

(event Custom|FundingAchieved
  ;[1 1 1] JLavender: Called when funds are added and it meets funding goal
  (bind _result (Math|Integer|IncrementInt (Variables|Default|GetUpdateFundsAchieved))
    (:%20%20
      (Variables|Default|SetUpdateFundsAchieved _result))))

(event Custom|UpdateSmokeValues)

(event Custom|OnFundingAchieved_Feedback
  (bind (_stationtypeint _smokeloc _smokesize) (CallFunction|FXGetSmokeCoverCueParams))
  ;[1 1 1] JLavender: Called when funds are added and it meets funding goal
  (GameplayCue|ExecuteGameplayCue%28Burst%29Local self (Variables|Funding|GetGCN_FundingAchieved) (Ability|GameplayCue|MakeGameplayCueParameters :NormalizedMagnitude (Math|Conversions|ToFloat(Integer) _stationtypeint) :RawMagnitude (Math|Conversions|ToFloat(Integer) _smokesize) :Location _smokeloc :Normal "0,0,0" :GameplayEffectLevel 1 :AbilityLevel 1 :bReplicateLocationWhenUsingMinimalRepProxy false)))

(event Custom|OnFundsAdded_Feedback
  ;[1 1 1] JLavender: Called when funds are added by players, but it does not reach the funding goal
  (GameplayCue|ExecuteGameplayCue%28Burst%29Local self (Variables|Funding|GetGCN_FundsAdded) (Ability|GameplayCue|MakeGameplayCueParameters 0.0)))

(event Custom|FundsAdded
  (bind _countloottovend (Variables|Default|GetCountLootToVend))
  (bind _returnvalue (> _countloottovend 0))
  ;[1 1 1] JLavender: Called when funds are added by players, but it does not reach the funding goal
  (if _returnvalue
    (while _returnvalue
      (Class|AthenaFunctionLibrary|SpawnLootfromTable (Variables|Default|GetDontationRewardLootName) 1 -1 (Transformation|GetSocketLocation (Variables|Debug|GetStaticMeshComponent) "Drop_Loot") (Variables|Default|GetLootSpawnDirection) 0 true true true -1 "Tossed" "ItemSpawner")
      (Math|Integer|DecrementInt (Variables|Default|GetCountLootToVend)))
    (bind _result (Math|Integer|IncrementInt (Variables|Default|GetUpdateFundsAdded))
      (:%20%20
        (Variables|Default|SetUpdateFundsAdded _result)))))

(fn AddToStationFund
  (params (FundingAmount :type Integer64) (FundChoice :type Integer))
  (Math|Integer|IncrementInt (Variables|Default|GetCountLootToVend)
    (:%20%20
      (Default|CallOnFundingReceived (Math|Float|TruncatetoInteger64 (* (Math|Conversions|ToInteger(Integer64) FundingAmount) (ScalableFloat|GetValueatLevel (Variables|Default|GetHotFixWhiteEggMonkey)))) (Variables|Default|GetFundingStationNum) FundChoice))))

(fn DestroyPreviewConstructionActors
  (Components|DestroyComponent (Variables|Default|GetStage01_Support_Prop_01)))

(fn FX_GetSmokeCoverCueParams
  (return 0 (Transformation|GetActorLocation) 0))

(fn GetFundingStationIndex
  (return (Variables|Default|GetFundingStationNum)))

(fn OnRep_Stage
  (switch int (Variables|Default|GetStage)
    (:0
      (CallFunction|NonexistentState))
    (:1
      (CallFunction|FundraisingState))
    (:2
      (CallFunction|FundedState))))

(fn OnRep_UpdateFundAchieved
  (CallFunction|UpdateSmokeValues)
  (Networking|CanExecuteCosmeticEvents
    (:True
      (CallFunction|OnFundingAchievedFeedback))))

(fn OnRep_UpdateFundsAdded
  (Networking|CanExecuteCosmeticEvents
    (:True
      (CallFunction|OnFundsAddedFeedback))))

(fn PlayServiceUsedEffects
  (params (PlayerController :type Fort%20Player%20Controller%20Object%20Reference))
  (|CallParentFunction PlayerController)
  (Utilities|IsValid PlayerController
    (:Is%20Valid
      (bind _asathena_player_controller (Utilities|Casting|CastToAthena_PlayerController PlayerController))
      (EventDispatchers|CallOnClientDonationSuccessDelegate _asathena_player_controller))))

(fn SetDressActors
  (Variables|Default|SetArrayOfSetDressActors (Utilities|Array|MakeArray (Variables|Default|GetStage01_Support_Prop_01) (Variables|Default|GetApollo_Farm_WheelBarrow_04) (Variables|Default|GetStage01_Support_Prop_05) (Variables|Default|GetStage01_Support_Prop_06) (Variables|Default|GetStage01_Support_Prop_02) (Variables|Default|GetStage01_Support_Prop_03))))

(fn ConstructionScript
  (|CallParentFunction))
