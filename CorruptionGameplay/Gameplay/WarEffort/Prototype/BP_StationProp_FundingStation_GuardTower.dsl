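; War effort guard tower funding station prop that replaces itself with a mounted turret and dressing props when funded.
; Funded, funding-achieved and spawn-turret events share one sequence, with authority knocking up nearby players and spawning the turret.
; Cleanup increments CleanUpRep, whose OnRep disables interaction, damages overlapping building actors and destroys the station.
; It also spawns dressing actors from child actor component classes and transforms, then removes a preview component.
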(event EventFundedState
  ;[1 1 1] Funded state start to clean up, knock players, spawn dress up actors and the mounted turret
  (CallFunction|SetDressActorsforMountedTurret)
  (CallFunction|CleanUp)
  (Networking|SwitchHasAuthority
    (:Authority
      (CallFunction|KnockbackPlayers)
      (Game|SpawnActorfromClass (Variables|Default|GetMountedTurret) (Math|Conversions|ToTransform(Vector) (- (Transformation|GetActorLocation) "0, 0,20.000000")) "Undefined" "SelectDefaultAtRuntime" 0 0))))

(event EventFundingAchieved
  ;[1 1 1] Funded state start to clean up, knock players, spawn dress up actors and the mounted turret
  (|CallParentFunction)
  (CallFunction|SetDressActorsforMountedTurret)
  (CallFunction|CleanUp)
  (Networking|SwitchHasAuthority
    (:Authority
      (CallFunction|KnockbackPlayers)
      (Game|SpawnActorfromClass (Variables|Default|GetMountedTurret) (Math|Conversions|ToTransform(Vector) (- (Transformation|GetActorLocation) "0, 0,20.000000")) "Undefined" "SelectDefaultAtRuntime" 0 0))))

(event Custom|SpawnWarEffortTurret
  ;[1 1 1] Funded state start to clean up, knock players, spawn dress up actors and the mounted turret
  (|CallParentFunction)
  (CallFunction|SetDressActorsforMountedTurret)
  (CallFunction|CleanUp)
  (Networking|SwitchHasAuthority
    (:Authority
      (CallFunction|KnockbackPlayers)
      (Game|SpawnActorfromClass (Variables|Default|GetMountedTurret) (Math|Conversions|ToTransform(Vector) (- (Transformation|GetActorLocation) "0, 0,20.000000")) "Undefined" "SelectDefaultAtRuntime" 0 0))))

(event Custom|CleanUp
  ;[1 1 1] Clean up
  (bind _result (Math|Integer|IncrementInt (Variables|Default|GetCleanUpRep))
    (:%20%20
      (Variables|Default|SetCleanUpRep _result))))

(event Custom|DestroyObjects
  ;[1 1 1] Destroy building actors in the area
  (bind _outactors (Collision|SphereOverlapActors (Transformation|GetActorLocation) (Variables|Default|GetDestructionRadius) (Variables|Default|GetObject%20Types) "/Script/FortniteGame.BuildingActor"))
  (for _array_element _outactors
    (Environment|ApplyGlobalEnvironmentDamagetoActor :DamageTargetActor _array_element :EnvironmentDamage (Variables|Default|GetEnvironment%20Damage) :EnvironmentDamageGE "/CorruptionGameplay/Gameplay/WarEffort/Prototype/GE_Athena_WarEffort_BuildUpDamage.GE_Athena_WarEffort_BuildUpDamage_C"))
  (Actor|DestroyActor))

(event Custom|Hide
  ;[1 1 1] Hide Prop
  (Rendering|SetVisibility (Variables|Debug|GetStaticMeshComponent) false false))

(event Custom|SpawnMountedTurretDressUpActors
  ;[1 1 1] Spawn Dress actors
  (for _array_element (Variables|Default|GetArrayOfSetDressUpActors_MountedTurret)
    (Utilities|IsValid _array_element
      (:Is%20Valid
        (if (Math|Vector|NotEqual(Vector) (.location (Transformation|GetWorldTransform _array_element)) "0, 0, 0" 0.0001)
          (Game|SpawnActorfromClass (Variables|/Script/Engine.ChildActorComponent|GetChildActorClass _array_element) (Transformation|GetWorldTransform _array_element) "AlwaysSpawn" "SelectDefaultAtRuntime" 0 0)))))
  (CallFunction|DestroyPreviewTowerActors))

(event Custom|StopInteractions
  ;[1 1 1] Prevent Interactions
  (Variables|Interaction|SetbAllowInteract false))

(fn DestroyPreviewTowerActors
  (Components|DestroyComponent (Variables|Default|GetStage01_Support_Prop_06)))

(fn FX_GetSmokeCoverCueParams
  (return 1 (Transformation|GetActorLocation) 0))

(fn KnockbackPlayers
  (Fortnite|ApplyGameplayEffectToActorsInRange (Utilities|Array|MakeArray "/CorruptionGameplay/Gameplay/WarEffort/GuardTower/GE_WarEffort_GuardTower_Grant_KnockUp.GE_WarEffort_GuardTower_Grant_KnockUp_C") (Utilities|Array|MakeArray "ObjectTypeQuery3") (Transformation|GetActorLocation) 6e+02))

(fn OnRep_CleanUpRep
  (CallFunction|StopInteractions)
  (CallFunction|DestroyObjects)
  (CallFunction|SpawnMountedTurretDressUpActors))

(fn SetDressActorsForMountedTurret
  (Variables|Default|SetArrayOfSetDressUpActors_MountedTurret (Utilities|Array|MakeArray (Variables|Default|GetApollo_Sandbag_03_FundedStage) (Variables|Default|GetApollo_MountainBase_GunCleaningKit_01_FundedStage) (Variables|Default|GetApollo_Sandbag_04_FundedStage) (Variables|Default|GetApollo_Sandbag_05_FundedStage) (Variables|Default|GetAthena_Prop_Military_Crate_FundedStage) (Variables|Default|GetApollo_Sandbag_02_FundedStage))))

(fn ConstructionScript
  (|CallParentFunction))
