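; War effort funding station prop where votes go toward one of two new weapons, using the funding manager's shared new-weapon funding index.
; A vote that completes funding tosses out loot for the chosen weapon, then triggers cleanup and sets the winner.
; Already-funded state events skip the loot.
; Rep-notifies disable interaction, remove collision and the conversation component, hide the mesh on clients, and swap check and X decals to show the result.
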(event EventFundedStateA
  ;[1 1 1] Funded States
  (Variables|Default|SetIsVoteA true)
  (CallFunction|CleanUp)
  (Variables|Default|SetWinner 1))

(event EventFundedStateB
  ;[1 1 1] Funded States
  (Variables|Default|SetIsVoteA false)
  (CallFunction|CleanUp)
  (Variables|Default|SetWinner 2))

(event EventVoteA
  ;[1 1 1] Funding for A and B and also Funded States for A and B when its initially completed
  (|CallParentFunction)
  (CallFunction|CheckFundingAmount)
  (Variables|Default|SetIsVoteA true)
  (if (Variables|Default|GetIsFunded)
    (CallFunction|SpawnLoot)
    (CallFunction|CleanUp)
    (Variables|Default|SetWinner 1)))

(event EventVoteB
  ;[1 1 1] Funding for A and B and also Funded States for A and B when its initially completed
  (|CallParentFunction)
  (CallFunction|CheckFundingAmount)
  (Variables|Default|SetIsVoteA false)
  (if (Variables|Default|GetIsFunded)
    (CallFunction|SpawnLoot)
    (CallFunction|CleanUp)
    (Variables|Default|SetWinner 2)))

(event Custom|SpawnLoot
  (bind _number_to_spawn (Variables|Default|GetNumber%20to%20Spawn))
  ;[1 1 1] Spawn a butt load of funded items
  (bind _counter (Utilities|FlowControl|DoN _number_to_spawn)
    (:Exit
      (Class|AthenaFunctionLibrary|SpawnLootfromTable (select (Variables|Default|GetIsVoteA) (Variables|Default|GetNewWeaponA) (Variables|Default|GetNewWeaponB)) 1 -1 (+ (Transformation|GetSocketLocation (Variables|Debug|GetStaticMeshComponent) "Drop_Loot") (Utilities|RandomVectorinRange :Max (Variables|Default|GetSpawnLocationOffset))) "0, 0, 0" 0 true true true -1 "Tossed" "ItemSpawner")
      (if (>= _counter _number_to_spawn)
        _
        (else
          _))
      (Networking|ForceNetUpdate))))

(event Custom|CleanUp
  ;[1 1 1] Clean up
  (bind _result (Math|Integer|IncrementInt (Variables|Default|GetCleanUpRep))
    (:%20%20
      (Variables|Default|SetCleanUpRep _result))))

(event Custom|Hide
  ;[1 1 1] Hide Mesh
  (Rendering|SetVisibility (Variables|Debug|GetStaticMeshComponent) false false))

(event Custom|RemoveBox
  ;[1 1 1] Clean up donation box interactions
  (Collision|SetCollisionEnabled (Variables|Debug|GetStaticMeshComponent) "NoCollision")
  (Components|DestroyComponent (Variables|Components|GetNonPlayerConversationComponent)))

(event Custom|MaterialOrDecalSwap
  (params (Winner :type Integer))
  (bind _decal_topleft (Variables|Default|GetDecal_TopLeft))
  (bind _decal_bottomright (Variables|Default|GetDecal_BottomRight))
  ;[1 1 1] Material Swap
  (switch int Winner
    (:1
      (Rendering|Components|Decal|SetDecalMaterial _decal_topleft (Variables|Default|GetCheck_Decal))
      (Rendering|Components|Decal|SetDecalMaterial (Variables|Default|GetDecal_BottomRight) (Variables|Default|GetX_Decal))
      (Development|SetHiddeninGame _decal_topleft false false))
    (:2
      (Rendering|Components|Decal|SetDecalMaterial _decal_bottomright (Variables|Default|GetCheck_Decal))
      (Rendering|Components|Decal|SetDecalMaterial (Variables|Default|GetDecal_TopLeft) (Variables|Default|GetX_Decal))
      (Development|SetHiddeninGame _decal_bottomright false false))))

(fn GetFundingStationIndex
  (return (Variables|/CorruptionGameplay/Gameplay/WarEffort/MeshNetwork/BP_FundingManager_Parent.BP_FundingManager_Parent_C|GetSharedNewWeaponFunding (Variables|Default|GetFundingManagerRef))))

(fn OnRep_CleanUpRep
  (Variables|Interaction|SetbAllowInteract false)
  (CallFunction|RemoveBox)
  (Networking|SwitchHasAuthority
    (:Remote
      (CallFunction|Hide))))

(fn OnRep_Winner
  (CallFunction|MaterialorDecalSwap (Variables|Default|GetWinner)))

(fn ConstructionScript
  (|CallParentFunction))
