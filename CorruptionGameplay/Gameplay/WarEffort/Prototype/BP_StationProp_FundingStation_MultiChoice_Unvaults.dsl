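; War effort funding station prop where players fund one of two competing item unvaults, built as a child of a multi-choice station.
; The active event index sets CurrentItem, whose RepNotify picks the pair's loot keys, icons, names and billboard material.
; A completing vote tosses the winner's loot; already-funded state events skip loot.
; Both set replicated CleanUpRep and Winner, which remove donation interaction and, on clients, hide the box and show check/X decals.
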(event EventSetupUnvaultItems
  ;[1 1 1] Setup which Unvaults should spawn
  (|CallParentFunction)
  (Variables|Default|SetCurrentItem (Utilities|Array|FindItem (Variables|Default|GetSetupEvents) (Variables|Default|GetCurrentActiveEvent))))

(event Custom|SetIcon&Name
  ;[1 1 1] Setup which Unvaults should spawn
  (bind _keys (Utilities|Map|Keys (Variables|Default|GetUnvaults_A)))
  (bind _keys_2 (Utilities|Map|Keys (Variables|Default|GetUnvaults_B)))
  (CallFunction|SetNewUnvaultTable (Utilities|Array|Get(acopy) _keys (Variables|Default|GetCurrentItem)) (Utilities|Array|Get(acopy) _keys_2 (Variables|Default|GetCurrentItem)) (Variables|Default|GetCurrentItem)))

(event Custom|TestUnvault
  (params (Key :type Integer))
  ;[1 0.3222 0.8165] Test to see which unvault is available
  (bind _keys (Utilities|Map|Keys (Variables|Default|GetUnvaults_A)))
  (bind _keys_2 (Utilities|Map|Keys (Variables|Default|GetUnvaults_B)))
  (CallFunction|SetNewUnvaultTable (Utilities|Array|Get(acopy) _keys Key) (Utilities|Array|Get(acopy) _keys_2 Key) Key)
  (Development|PrintString (Utilities|GetDisplayName (Utilities|ResolveSoftReference (Variables|Default|GetFundingReferenceA))))
  (Development|PrintString (Utilities|GetDisplayName (Utilities|ResolveSoftReference (Variables|Default|GetFundingReferenceB)))))

(event EventVoteA
  ;[1 1 1] Funding for A and B and also Funded States for A and B when its initially completed
  (|CallParentFunction)
  (CallFunction|CheckFundingAmount)
  (Variables|Default|SetIsVoteA true)
  (if (Variables|Default|GetIsFunded)
    (CallFunction|SpawnLoot)
    (CallFunction|CleanUp)
    (Variables|Default|SetWinner 1)))

(event EventVoteB
  ;[1 1 1] Funding for A and B and also Funded States for A and B when its initially completed
  (|CallParentFunction)
  (CallFunction|CheckFundingAmount)
  (Variables|Default|SetIsVoteA false)
  (if (Variables|Default|GetIsFunded)
    (CallFunction|SpawnLoot)
    (CallFunction|CleanUp)
    (Variables|Default|SetWinner 2)))

(event Custom|SpawnLoot
  (bind _number_to_spawn (Variables|Default|GetNumber%20to%20Spawn))
  ;[1 1 1] Comment
  (bind _counter (Utilities|FlowControl|DoN _number_to_spawn)
    (:Exit
      ;;[0.3802 0.3353 0] Needs to pick loot based on which one won.
      (Class|AthenaFunctionLibrary|SpawnLootfromTable (select (Variables|Default|GetIsVoteA) (Variables|Default|GetUnvaultA) (Variables|Default|GetUnvaultB)) 1 -1 (+ (Transformation|GetSocketLocation (Variables|Debug|GetStaticMeshComponent) "Drop_Loot") (Utilities|RandomVectorinRange :Max (Variables|Default|GetSpawnLocationOffset))) "0, 0, 0" 0 true true true -1 "Tossed" "ItemSpawner")
      (if (>= _counter _number_to_spawn)
        _
        (else
          _))
      (Networking|ForceNetUpdate))))

(event Custom|CleanUp
  ;[1 1 1] Clean up
  (bind _result (Math|Integer|IncrementInt (Variables|Default|GetCleanUpRep))
    (:%20%20
      (Variables|Default|SetCleanUpRep _result))))

(event Custom|RemoveBox
  ;[1 1 1] Clean up donation box interactions
  (Collision|SetCollisionEnabled (Variables|Debug|GetStaticMeshComponent) "NoCollision")
  (Components|DestroyComponent (Variables|Components|GetNonPlayerConversationComponent)))

(event Custom|Hide
  ;[1 1 1] Hide Donation Box
  (Rendering|SetVisibility (Variables|Debug|GetStaticMeshComponent) false false))

(event EventFundedStateA
  ;[1 1 1] Funded States
  (Variables|Default|SetIsVoteA true)
  (CallFunction|CleanUp)
  (Variables|Default|SetWinner 1))

(event EventFundedStateB
  ;[1 1 1] Funded States
  (Variables|Default|SetIsVoteA false)
  (CallFunction|CleanUp)
  (Variables|Default|SetWinner 2))

(event Custom|MaterialOrDecalSwap
  (params (Winner :type Integer))
  (bind _decal_bottomright (Variables|Default|GetDecal_BottomRight))
  (bind _decal_topleft (Variables|Default|GetDecal_TopLeft))
  ;[1 1 1] Swap Material
  (switch int Winner
    (:1
      (Rendering|Components|Decal|SetDecalMaterial _decal_topleft (Variables|Default|GetCheck_Decal))
      (Rendering|Components|Decal|SetDecalMaterial (Variables|Default|GetDecal_BottomRight) (Variables|Default|GetX_Decal))
      (Development|SetHiddeninGame _decal_topleft false false))
    (:2
      (Rendering|Components|Decal|SetDecalMaterial _decal_bottomright (Variables|Default|GetCheck_Decal))
      (Rendering|Components|Decal|SetDecalMaterial (Variables|Default|GetDecal_TopLeft) (Variables|Default|GetX_Decal))
      (Development|SetHiddeninGame _decal_bottomright false false))))

(fn GetFundingStationIndex
  (return (Variables|/CorruptionGameplay/Gameplay/WarEffort/MeshNetwork/BP_FundingManager_Parent.BP_FundingManager_Parent_C|GetSharedUnvaultFunding (Variables|Default|GetFundingManagerRef))))

(fn OnRep_CleanUpRep
  (Variables|Interaction|SetbAllowInteract false)
  (CallFunction|RemoveBox)
  (Networking|SwitchHasAuthority
    (:Remote
      (CallFunction|Hide))))

(fn OnRep_CurrentItem
  (CallFunction|SetIcon&Name))

(fn OnRep_Winner
  (Networking|SwitchHasAuthority
    (:Remote
      (CallFunction|MaterialorDecalSwap (Variables|Default|GetWinner)))))

(fn SetNewUnvaultTable
  (params (UnvaultA :type Name) (UnvaultB :type Name) (ArrayIndex :type Integer))
  (Variables|Default|SetUnvaultA UnvaultA)
  (Variables|Default|SetUnvaultB UnvaultB)
  (Variables|Default|SetFundingReferenceA (Utilities|Select :Index ArrayIndex :Option%200 "/CorruptionGameplay/Gameplay/WarEffort/Prototype/Textures/Icons/T-Icon-Weapons-ShockwaveGrenadeLauncher-L.T-Icon-Weapons-ShockwaveGrenadeLauncher-L" :Option%201 "/Game/UI/Foundation/Textures/Icons/Athena/T-T-Icon-BR-SM-Athena-DanceGrenade-L.T-T-Icon-BR-SM-Athena-DanceGrenade-L" :Option%202 "/Game/UI/Foundation/Textures/Icons/Weapons/Items/T-Icon-Weapons-ProximityGrenadeLauncher-L.T-Icon-Weapons-ProximityGrenadeLauncher-L"))
  (Variables|Default|SetFundingReferenceB (Utilities|Select :Index ArrayIndex :Option%200 "/Game/UI/Foundation/Textures/Icons/Athena/T-Icon-BR-Rift-Globe-L.T-Icon-BR-Rift-Globe-L" :Option%201 "/Game/UI/Foundation/Textures/Icons/Weapons/Items/T-Icon-Weapons-TacticalShotgun02-L.T-Icon-Weapons-TacticalShotgun02-L" :Option%202 "/Game/UI/Foundation/Textures/Icons/Weapons/Items/T-Icon-Weapons-SK-Flintlock-Pistol-Single-Shot-L.T-Icon-Weapons-SK-Flintlock-Pistol-Single-Shot-L"))
  (Variables|Default|SetFundingTextA (Utilities|Select :Index ArrayIndex :Option%200 "NSLOCTEXT(\"[BB78D0D443D57F1914498C814A66912B]\", \"5C0841964EDB1E5EE9F5BE9DEAD6C263\", \"Shockwave Launcher\")" :Option%201 "NSLOCTEXT(\"[BB78D0D443D57F1914498C814A66912B]\", \"6673809E4323CCDB2A8A309648EF9B6F\", \"Boogie Bomb\")" :Option%202 "NSLOCTEXT(\"[BB78D0D443D57F1914498C814A66912B]\", \"6AB3207F44A09D465FAE68A52290B375\", \"Proximity Grenade Launcher\")"))
  (Variables|Default|SetFundingTextB (Utilities|Select :Index ArrayIndex :Option%200 "NSLOCTEXT(\"[BB78D0D443D57F1914498C814A66912B]\", \"9B4F452D47E9CCBFEAAB5BB2C7335CD9\", \"Rift-To-Go\")" :Option%201 "NSLOCTEXT(\"[BB78D0D443D57F1914498C814A66912B]\", \"516152DA4916B72FD83F2C8FBE4ED625\", \"Combat Shotgun\")" :Option%202 "NSLOCTEXT(\"[BB78D0D443D57F1914498C814A66912B]\", \"61E975DF41CD733AABCD5982712613F1\", \"Flint-Knock Pistol\")"))
  (Rendering|Material|SetMaterial (Variables|Default|GetBillBoard) 1 (Utilities|Select :Index ArrayIndex :Option%200 "/Game/Environments/Apollo/Sets/Quest/Props/Materials/MI_Apollo_Quest_FundingStation_BulletinBoard_Board_01_A.MI_Apollo_Quest_FundingStation_BulletinBoard_Board_01_A" :Option%201 "/Game/Environments/Apollo/Sets/Quest/Props/Materials/MI_Apollo_Quest_FundingStation_BulletinBoard_Board_01_B.MI_Apollo_Quest_FundingStation_BulletinBoard_Board_01_B" :Option%202 "/Game/Environments/Apollo/Sets/Quest/Props/Materials/MI_Apollo_Quest_FundingStation_BulletinBoard_Board_01_C.MI_Apollo_Quest_FundingStation_BulletinBoard_Board_01_C")))

(fn ConstructionScript
  (|CallParentFunction))
