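; War effort funding station prop (prototype) that, once funded, enables a launch pad building foundation and spawns a building actor high above it.
; Funding events trigger cleanup and knock nearby players up with a gameplay effect.
; Cleanup increments a replicated counter whose OnRep hides and destroys the prop, stops interaction and damages surrounding building actors.
; BeginPlay checks per-location hotfix values, though only the first location calls parent logic; smoke size depends on the launch pad.
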(event EventBeginPlay
  ;[1 1 1] Hotfix for each mech station location
  (switch int (Variables|Default|GetMechLocation)
    (:0
      (if (not (> (ScalableFloat|GetValueatLevel (Variables|Default|GetPleasentParkHotfixOff)) 0.0))
        (|CallParentFunction)))
    (:1
      (if (> (ScalableFloat|GetValueatLevel (Variables|Default|GetCornyComplexHotfixOff)) 0.0)
        _
        (else
          _)))
    (:2
      (if (> (ScalableFloat|GetValueatLevel (Variables|Default|GetDefiantDishHotfixOff)) 0.0)
        _
        (else
          _))))
  (Variables|Default|SetStage 0))

(event EventUpdateSmokeValues
  ;[1 1 1] Update Smoke values based on level bounds
  (bind _output_get (Variables|Default|SetSmokeSize (Math|Integer|SelectInt 1 0 (Utilities|Equal(Object) (Utilities|ResolveSoftReference (Variables|Default|GetLaunchPad)) (Utilities|ResolveSoftReference (Variables|Default|GetLargerLaunchPad))))))
  (Development|PrintString (Utilities|String|ToString(Integer) _output_get) true true "(R=0.000000,G=0.660000,B=1.000000,A=1.000000)" 1e+01))

(event EventFundedState
  (bind _launchpad (Variables|Default|GetLaunchPad))
  (bind _output (Utilities|ResolveSoftReference _launchpad))
  ;[1 1 1] Funded state , dynamically load level , clean up and spawn new thing
  (CallFunction|UpdateSmokeValues)
  (CallFunction|CleanUp)
  (CallFunction|KnockbackPlayers)
  (Class|BuildingFoundation|SetDynamicFoundationEnabled _output true)
  (Variables|/Script/FortniteGame.BuildingFoundation|SetDynamicFoundationType "StartEnabled_Dynamic" _output)
  (Building|SpawnBuildingActor (Variables|Default|GetBGA_Thing) (Math|Transform|MakeTransform (Math|Vector|MakeVector (.x (.location (Transformation|GetWorldTransform (Variables|/Script/Engine.Actor|GetRootComponent _output)))) (.y (.location (Transformation|GetWorldTransform (Variables|/Script/Engine.Actor|GetRootComponent _output)))) (+ (.z (.location (Transformation|GetWorldTransform (Variables|/Script/Engine.Actor|GetRootComponent _output)))) 5e+04)) (.rotation (Transformation|GetWorldTransform (Variables|/Script/Engine.Actor|GetRootComponent _output))) (.scale (Transformation|GetWorldTransform (Variables|/Script/Engine.Actor|GetRootComponent _output))))))

(event EventFundingAchieved
  (bind _launchpad (Variables|Default|GetLaunchPad))
  (bind _output (Utilities|ResolveSoftReference _launchpad))
  ;[1 1 1] Funded state , dynamically load level , clean up and spawn new thing
  (|CallParentFunction)
  (CallFunction|CleanUp)
  (CallFunction|KnockbackPlayers)
  (Class|BuildingFoundation|SetDynamicFoundationEnabled _output true)
  (Variables|/Script/FortniteGame.BuildingFoundation|SetDynamicFoundationType "StartEnabled_Dynamic" _output)
  (Building|SpawnBuildingActor (Variables|Default|GetBGA_Thing) (Math|Transform|MakeTransform (Math|Vector|MakeVector (.x (.location (Transformation|GetWorldTransform (Variables|/Script/Engine.Actor|GetRootComponent _output)))) (.y (.location (Transformation|GetWorldTransform (Variables|/Script/Engine.Actor|GetRootComponent _output)))) (+ (.z (.location (Transformation|GetWorldTransform (Variables|/Script/Engine.Actor|GetRootComponent _output)))) 5e+04)) (.rotation (Transformation|GetWorldTransform (Variables|/Script/Engine.Actor|GetRootComponent _output))) (.scale (Transformation|GetWorldTransform (Variables|/Script/Engine.Actor|GetRootComponent _output))))))

(event Custom|CleanUp
  ;[1 1 1] Clean up
  (bind _result (Math|Integer|IncrementInt (Variables|Default|GetCleanUpRep))
    (:%20%20
      (Variables|Default|SetCleanUpRep _result))))

(event Custom|DestroyObjects
  ;[1 1 1] Destroy building actors in the area
  (bind _outactors (Collision|SphereOverlapActors (Transformation|GetActorLocation) (Variables|Default|GetDestructionRadius) (Variables|Default|GetObject%20Types) "/Script/FortniteGame.BuildingActor"))
  (for _array_element _outactors
    (Environment|ApplyGlobalEnvironmentDamagetoActor :DamageTargetActor _array_element :EnvironmentDamage (Variables|Default|GetEnvironment%20Damage) :EnvironmentDamageGE "/CorruptionGameplay/Gameplay/WarEffort/Prototype/GE_Athena_WarEffort_BuildUpDamage.GE_Athena_WarEffort_BuildUpDamage_C")))

(event Custom|Hide
  ;[1 1 1] Hide and destroy self
  (Rendering|SetVisibility (Variables|Debug|GetStaticMeshComponent) false true)
  (Actor|DestroyActor))

(event Custom|StopInteractions
  ;[1 1 1] Prevent interactions to the box
  (Components|DestroyComponent (Variables|Components|GetNonPlayerConversationComponent))
  (Variables|Interaction|SetbAllowInteract false))

(fn FX_GetSmokeCoverCueParams
  (return 2 (Math|Vector|MakeVector (.x (Transformation|GetActorLocation (Utilities|ResolveSoftReference (Variables|Default|GetLaunchPad)))) (.y (Transformation|GetActorLocation (Utilities|ResolveSoftReference (Variables|Default|GetLaunchPad)))) (.z (Transformation|GetWorldLocation (Variables|Debug|GetStaticMeshComponent)))) (Variables|Default|GetSmokeSize)))

(fn KnockbackPlayers
  (Fortnite|ApplyGameplayEffectToActorsInRange (Utilities|Array|MakeArray "/CorruptionGameplay/Gameplay/WarEffort/GuardTower/GE_WarEffort_GuardTower_Grant_KnockUp.GE_WarEffort_GuardTower_Grant_KnockUp_C") (Utilities|Array|MakeArray "ObjectTypeQuery3") (Transformation|GetActorLocation) 6e+02))

(fn OnRep_CleanUpRep
  (CallFunction|Hide)
  (CallFunction|StopInteractions)
  (CallFunction|DestroyObjects))

(fn OnRep_TurnOffInteract
  (CallFunction|StopInteractions))

(fn ConstructionScript
  (|CallParentFunction))
